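# Physics asset for the Dino Collector backpack cosmetic, defining collision bodies and joints for the pack and its dangling pieces.
# A box covers the pack, with box bodies for two swords and two necklace segments, plus a capsule for the pendant.
# Both swords hinge off the pack with tight swing and twist limits.
# The necklace chains from pack through mid segment to pendant with looser swing and little twist.
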PhysicsAsset /BRCosmetics/Accessories/FORT_Backpacks/Backpack_F_MED_Dino_Collector/Meshes/F_MED_Dino_Collector_Pack_Physics
bodies {
 dyn_necklace [{shapeType=Box, extentX=4.0339, extentY=1.1014, extentZ=7.6924, center=<2.6577 0.0002 0.0001>}]
 dyn_necklacePendant [{shapeType=Capsule, length=2.2627, radius=1.0552, center=<1.2269 -0.7212 0.0417>, rotation=<rot 0 -92.1965 -89.2284>}]
 dyn_necklace_mid [{shapeType=Box, extentX=3.0402, extentY=1.0484, extentZ=4.6984, center=<2.0099 0 0>}]
 dyn_sword_l [{shapeType=Box, extentX=34.816, extentY=2.476, extentZ=6.123, center=<1.1586 0.0001 0>}]
 dyn_sword_r [{shapeType=Box, extentX=38.3096, extentY=3.5788, extentZ=5.7253, center=<1.1797 0 0>}]
 pack [{shapeType=Box, extentX=29.2207, extentY=7.5304, extentZ=29.5011, center=<0.0096 -2.0122 -8.1768>}]
}
constraints [
 {bone1Name=dyn_sword_l, bone2Name=pack, swing1LimitDegrees=10, swing2LimitDegrees=10, twistLimitDegrees=1}
 {bone1Name=dyn_sword_r, bone2Name=pack, swing1LimitDegrees=10, swing2LimitDegrees=10, twistLimitDegrees=1}
 {bone1Name=dyn_necklace, bone2Name=pack, swing1LimitDegrees=20, swing2LimitDegrees=50, twistLimitDegrees=2}
 {bone1Name=dyn_necklacePendant, bone2Name=dyn_necklace_mid, swing1LimitDegrees=25, swing2LimitDegrees=55, twistLimitDegrees=5}
 {bone1Name=dyn_necklace_mid, bone2Name=dyn_necklace, swing1LimitDegrees=25, swing2LimitDegrees=50, twistLimitDegrees=2}
]
